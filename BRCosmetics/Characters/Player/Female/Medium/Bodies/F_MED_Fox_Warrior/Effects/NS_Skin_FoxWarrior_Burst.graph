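# Electric burst effect for the Fox Warrior skin, played on the character's skeletal mesh with a pink or blue look chosen by VariantNum.
# Rapidly looping emitters draw short ribbon arcs with glow sprites between sampled bones and sockets, the bone arcs adding lights.
# Each arc shares a random spawn group as ribbon ID and seed, interpolates between endpoints and bows outward via a sine mask.
# Mesh-following glow sprites and flung sparkles complete the burst.
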NiagaraSystem /BRCosmetics/Characters/Player/Female/Medium/Bodies/F_MED_Fox_Warrior/Effects/NS_Skin_FoxWarrior_Burst
dependencies {
 usedDataInterfaces [&NiagaraDataInterfaceArrayInt32, &NiagaraDataInterfaceCurve]
 usedDynamicInputs []
 usedModules [&SystemState]
}
emitters {
 Electric_Bone_Explosion {
  emitterSpawnScript [/SetVariables(Emitter.MaxEmitterLoops=20) {isSetParameters=true, module=SetVariables_#1}]
  emitterUpdateScript [
   &EmitterState(
    Life Cycle Mode=::NewEnumerator1
    Loop Behavior=::NewEnumerator2
    Loop Count:NiagaraInt32=$Emitter.MaxEmitterLoops
    Loop Duration=&RandomRangeFloat(Minimum=0.025, Maximum=0.045, Evaluation Type=::NewEnumerator1)
    Recalculate Duration Each Loop=true
    UseLoopDelay=false
   )
   &SpawnBurst_Instantaneous(
    Spawn Count=6
    Spawn Group:NiagaraInt32=&RandomRangeInt(Minimum=0, Maximum=9999, Evaluation Type=::NewEnumerator1)
    Use Spawn Probability=false
   )
  ]
  particleSpawnScript [
   /SetVariables(Particles.StoredNormalizedExec:NiagaraFloat=&ReturnNormalizedExecIndex()) {isSetParameters=true, module=SetVariables_#2}
   /SetVariables(Particles.ParticleCurveMask:NiagaraFloat=&Sine(Normalized Angle=&ReturnNormalizedExecIndex(), Period=2, Scale=1, Bias=0)) {isSetParameters=true, module=SetVariables_#3}
   /SetVariables(
    Particles.BoneRange:NiagaraInt32=&RandomRangeInt(Minimum=20, Maximum=110)
    Particles.ElectricHeight:NiagaraFloat=&RandomRangeFloat(Minimum=-8, Maximum=8, Randomness Mode=::NewEnumerator2, Random Seed=&SpawnGroup(), Fixed Random Seed=true)
    Particles.ElectricSlide:NiagaraFloat=&Multiply_Float(
     A=&RandomRangeFloat(Minimum=0, Maximum=3, Randomness Mode=::NewEnumerator2, Random Seed=&SpawnGroup(), Fixed Random Seed=true)
     B=$Particles.ParticleCurveMask
    )
   ) {isSetParameters=true, module=SetVariables_#4}
   /SetVariables(
    Particles.SpriteColor=&MakeLinearColorFromVectorAndFloat(
     "Vector (RGB)"=&MakeVectorFromLinearColorRGB(Color=&LerpLinearColors(
       StartColor=<rgba 1 0 0.6 1>
       EndColor=<rgba 0 0.6333 1 1>
       LerpFactor=&MakeFloatFromInt(Int=$User.VariantNum)
      ))
     "Float (Alpha)"=&RandomRangeFloat(Minimum=0.125, Maximum=0.15)
    )
    Particles.SpriteOffset:Vector3f=<3 0 0>
    Particles.SpriteSize=&Vector2DFromFloat(Value=&UniformAOrBFloat(A=30, B=0, Distribution Weight=&Ceil_Float(Float=$Particles.ParticleCurveMask)))
   ) {isSetParameters=true, module=SetVariables_#5}
   &InitializeRibbon(
    Color=&MakeLinearColor(
     R:NiagaraFloat=&RandomRangeFloat(Minimum=1, Maximum=4)
     G:NiagaraFloat=1
     B:NiagaraFloat=0
     A:NiagaraFloat=0
    )
    Lifetime=&RandomRangeFloat(Minimum=0.1, Maximum=0.1, Randomness Mode=::NewEnumerator1, Random Seed=$Emitter.LoopCount, Fixed Random Seed=true)
    Module.Write Color=true
    Module.Write Lifetime=true
    Module.Write Mass=false
    Module.Write Material Random=true
    Module.Write Position=false
    Module.Write RibbonTwist=false
    Module.Write RibbonWidth=true
    Ribbon Width=&RandomRangeFloat(Minimum=3.5, Maximum=5)
   )
   /SetVariables(Particles.RibbonID=/Niagara/DynamicInputs/NiagaraID/MakeNiagaraIDFromInt(Index:NiagaraInt32=&SpawnGroup(), Acquire Tag:NiagaraInt32=0)) {isSetParameters=true, module=SetVariables_#6}
   &DynamicMaterialParameters(
    Index 0 Param 1=1
    Index 1 Param 4=6.666
    Module.Param0WriteEnabled=true
    Module.Param0WriteEnabled1=true
    Module.Param1WriteEnabled=true
    Module.Param1WriteEnabled1=true
    Module.Param2WriteEnabled=true
    Module.Param2WriteEnabled1=true
    Module.Param3WriteEnabled=true
    Module.Param3WriteEnabled1=true
    Write Parameter Index 1=true
   )
   &CameraOffset(Camera Offset Amount=50)
   &SkeletalMeshLocation(
    Apply To Particle Position=false
    Bone / Socket Index=$Particles.BoneRange
    Bone Sampling Mode=::NewEnumerator6
    Inherit Velocity=false
    MaskSpawnGroup=false
    Skeletal Mesh=$User.SkeletalMesh
    Use Sampled Mesh Attribute as Ribbon ID=false
   )
   /SetVariables(Particles.StoredSourceLocation:Vector3f=$Output.SkeletalMeshLocation.SampledPosition) {isSetParameters=true, module=SetVariables_#7}
   &SkeletalMeshLocation(
    Apply To Particle Position=false
    Bone / Socket Index=$Particles.BoneRange
    Bone Sampling Mode=::NewEnumerator6
    Inherit Velocity=false
    MaskSpawnGroup=false
    Skeletal Mesh=$User.SkeletalMesh
    Use Sampled Mesh Attribute as Ribbon ID=false
   ) {module=SkeletalMeshLocation002}
   /SetVariables(Particles.StoredDestinationLocation:Vector3f=$Output.SkeletalMeshLocation002.SampledPosition) {isSetParameters=true, module=SetVariables_#8}
  ]
  particleUpdateScript [
   &ParticleState()
   /SetVariables(Particles.RibbonPosition:Vector3f=&Add_Vector(
     A=&Multiply_VectorByFloat(
      Vector=&Add_Vector(
       A=&MakeVector(X=0, Y=0, Z=$Particles.Initial.ElectricHeight)
       B=&Multiply_VectorByFloat(
        Vector=&ReturnMeshOrientationAxis(Mesh Orientation Axis=::XAxis, Mesh Orientation Quat=$Output.SkeletalMeshLocation.SampledRotation)
        Float=$Particles.Initial.ElectricSlide
       )
      )
      Float=&Multiply_Float(
       A=&ClampFloat(Float=&Power(a=$Particles.NormalizedAge, b=0.75), Min=1)
       B=$Particles.Initial.ParticleCurveMask
      )
     )
     B=&Lerp_Vector(A=$Particles.StoredSourceLocation, B=$Particles.StoredDestinationLocation, Alpha=$Particles.StoredNormalizedExec)
    )) {isSetParameters=true, module=SetVariables_#9}
   /SetVariables(Particles.SpritePosition:Vector3f=&Add_Vector(A=$Particles.Initial.SpriteOffset, B=$Particles.RibbonPosition)) {isSetParameters=true, module=SetVariables_#10}
   /SetVariables(Particles.SpriteColor=&MultiplyLinearColor(
     Linear Color=$Particles.Initial.SpriteColor
     Scale Factor=&MakeVector4_FromFloat(Float=&FloatFromCurve(FloatCurve=DataInterface() {propertyValues {Curve {keys [
           {arriveTangent=-2.4415, interpMode=RCIM_Cubic, time=0, value=0}
           <at 0.175 1 RCIM_Cubic>
           {arriveTangent=-3.0397, interpMode=RCIM_Cubic, leaveTangent=-3.0397, tangentMode=RCTM_User, time=1, value=0}
          ]}}}))
    )) {isSetParameters=true, module=SetVariables_#11}
   &ScaleRibbonWidth(Ribbon Width Scale=&FloatFromCurve(FloatCurve=DataInterface() {propertyValues {Curve {keys [
         {arriveTangent=-4.302, interpMode=RCIM_Cubic, time=0, value=1}
         {arriveTangent=-2.5253, interpMode=RCIM_Cubic, leaveTangent=-11.4402, tangentMode=RCTM_User, time=1, value=0}
        ]}}}))
   /SetVariables(
    Particles.RibbonWidthBlue:NiagaraFloat=&Multiply_Float(
     A=$Particles.RibbonWidth
     B=&Lerp_Float(A=0, B=1, Alpha=&MakeFloatFromInt(Int=$User.VariantNum))
    )
    Particles.RibbonWidthPink:NiagaraFloat=&Multiply_Float(
     A=$Particles.RibbonWidth
     B=&Lerp_Float(A=1, B=0, Alpha=&MakeFloatFromInt(Int=$User.VariantNum))
    )
   ) {isSetParameters=true, module=SetVariables_#12}
   /SetVariables(
    Particles.LightColor=&MultiplyLinearColor(Linear Color=$Particles.SpriteColor, Scale Factor=&MakeVector4_FromFloat(Float=2))
    Particles.LightEnabled:NiagaraBool=&SetBoolByFloatComparison(
     A=$Particles.Initial.ParticleCurveMask
     B=&Lerp_Float(
      A=&RandomRangeFloat(Minimum=0.95, Maximum=0.952)
      B=1
      Alpha=/Niagara/DynamicInputs/Integer/V2/Round_Float(Float:NiagaraFloat=$Particles.Initial.MaterialRandom)
     )
    )
    Particles.LightPosition:Vector3f=&Add_Vector(A=$Particles.SpritePosition, B=<0 0 20>)
    Particles.LightRadius=40
   ) {isSetParameters=true, module=SetVariables_#13}
  ]
  rendererClasses [
   &NiagaraRibbonRendererProperties
   &NiagaraSpriteRendererProperties
   &NiagaraLightRendererProperties
  ]
  renderers [
   {rendererClass=&NiagaraRibbonRendererProperties}
   {rendererClass=&NiagaraSpriteRendererProperties, rendererIndex=1}
   {rendererClass=&NiagaraLightRendererProperties, rendererIndex=2}
   {rendererClass=&NiagaraRibbonRendererProperties, rendererIndex=3}
  ]
 }
 Electric_Bone_Explosion001 {
  emitterSpawnScript [/SetVariables(Emitter.MaxEmitterLoops=2) {isSetParameters=true, module=SetVariables_#1}]
  emitterUpdateScript [
   &EmitterState(
    Life Cycle Mode=::NewEnumerator1
    Loop Behavior=::NewEnumerator2
    Loop Count:NiagaraInt32=$Emitter.MaxEmitterLoops
    Loop Duration=&RandomRangeFloat(Minimum=0.04, Maximum=0.15, Evaluation Type=::NewEnumerator1)
    Recalculate Duration Each Loop=true
    UseLoopDelay=false
   )
   &SpawnBurst_Instantaneous(
    Spawn Count=6
    Spawn Group:NiagaraInt32=&RandomRangeInt(Minimum=0, Maximum=9999, Evaluation Type=::NewEnumerator1)
    Use Spawn Probability=false
   )
  ]
  particleSpawnScript [
   /SetVariables(Particles.StoredNormalizedExec:NiagaraFloat=&ReturnNormalizedExecIndex()) {isSetParameters=true, module=SetVariables_#2}
   /SetVariables(Particles.ParticleCurveMask:NiagaraFloat=&Sine(Normalized Angle=&ReturnNormalizedExecIndex(), Period=2, Scale=1, Bias=0)) {isSetParameters=true, module=SetVariables_#3}
   /SetVariables(
    Particles.BoneRange:NiagaraInt32=&RandomRangeInt(Minimum=40, Maximum=100)
    Particles.ElectricHeight:NiagaraFloat=&RandomRangeFloat(Minimum=-8, Maximum=8, Randomness Mode=::NewEnumerator2, Random Seed=&SpawnGroup(), Fixed Random Seed=true)
    Particles.ElectricSlide:NiagaraFloat=&Multiply_Float(
     A=&RandomRangeFloat(Minimum=0, Maximum=3, Randomness Mode=::NewEnumerator2, Random Seed=&SpawnGroup(), Fixed Random Seed=true)
     B=$Particles.ParticleCurveMask
    )
   ) {isSetParameters=true, module=SetVariables_#4}
   /SetVariables(
    Particles.SpriteColor=&MakeLinearColorFromVectorAndFloat(
     "Vector (RGB)"=&MakeVectorFromLinearColorRGB(Color=&LerpLinearColors(
       StartColor=<rgba 1 0 0.6 1>
       EndColor=<rgba 0 0.6333 1 1>
       LerpFactor=&MakeFloatFromInt(Int=$User.VariantNum)
      ))
     "Float (Alpha)"=&RandomRangeFloat(Minimum=0.125, Maximum=0.15)
    )
    Particles.SpriteOffset:Vector3f=<3 0 0>
    Particles.SpriteSize=&Vector2DFromFloat(Value=&UniformAOrBFloat(A=30, B=0, Distribution Weight=&Ceil_Float(Float=$Particles.ParticleCurveMask)))
   ) {isSetParameters=true, module=SetVariables_#5}
   &InitializeRibbon(
    Color=&MakeLinearColor(
     R:NiagaraFloat=&RandomRangeFloat(Minimum=30, Maximum=50)
     G:NiagaraFloat=1
     B:NiagaraFloat=0
     A:NiagaraFloat=0
    )
    Lifetime=0.15
    Module.Write Color=true
    Module.Write Lifetime=true
    Module.Write Mass=false
    Module.Write Material Random=true
    Module.Write Position=false
    Module.Write RibbonTwist=false
    Module.Write RibbonWidth=true
    Ribbon Width=&RandomRangeFloat(Minimum=9, Maximum=10)
   )
   /SetVariables(Particles.RibbonID=&MakeNiagaraIDFromInt(Index:NiagaraInt32=&SpawnGroup(), Acquire Tag:NiagaraInt32=0)) {isSetParameters=true, module=SetVariables_#6}
   &DynamicMaterialParameters(
    Index 0 Param 1=1
    Index 1 Param 4=6.666
    Module.Param0WriteEnabled=true
    Module.Param0WriteEnabled1=true
    Module.Param1WriteEnabled=true
    Module.Param1WriteEnabled1=true
    Module.Param2WriteEnabled=true
    Module.Param2WriteEnabled1=true
    Module.Param3WriteEnabled=true
    Module.Param3WriteEnabled1=true
    Write Parameter Index 1=true
   )
   &CameraOffset(Camera Offset Amount=50)
   &SkeletalMeshLocation(
    Apply To Particle Position=false
    Bone / Socket Index=$Particles.BoneRange
    Bone Sampling Mode=::NewEnumerator6
    Inherit Velocity=false
    MaskSpawnGroup=false
    Skeletal Mesh=$User.SkeletalMesh
    Use Sampled Mesh Attribute as Ribbon ID=false
   )
   /SetVariables(Particles.StoredSourceLocation:Vector3f=$Output.SkeletalMeshLocation.SampledPosition) {isSetParameters=true, module=SetVariables_#7}
   &SkeletalMeshLocation(
    Apply To Particle Position=false
    Bone / Socket Index=$Particles.BoneRange
    Bone Sampling Mode=::NewEnumerator6
    Inherit Velocity=false
    MaskSpawnGroup=false
    Skeletal Mesh=$User.SkeletalMesh
    Use Sampled Mesh Attribute as Ribbon ID=false
   ) {module=SkeletalMeshLocation002}
   /SetVariables(Particles.StoredDestinationLocation:Vector3f=$Output.SkeletalMeshLocation002.SampledPosition) {isSetParameters=true, module=SetVariables_#8}
  ]
  particleUpdateScript [
   &ParticleState()
   /SetVariables(Particles.RibbonPosition:Vector3f=&Add_Vector(
     A=&Multiply_VectorByFloat(
      Vector=&Add_Vector(
       A=&MakeVector(X=0, Y=0, Z=$Particles.Initial.ElectricHeight)
       B=&Multiply_VectorByFloat(
        Vector=&ReturnMeshOrientationAxis(Mesh Orientation Axis=::XAxis, Mesh Orientation Quat=$Output.SkeletalMeshLocation.SampledRotation)
        Float=$Particles.Initial.ElectricSlide
       )
      )
      Float=&Multiply_Float(
       A=&ClampFloat(Float=&Power(a=$Particles.NormalizedAge, b=0.75), Min=1)
       B=$Particles.Initial.ParticleCurveMask
      )
     )
     B=&Lerp_Vector(A=$Particles.StoredSourceLocation, B=$Particles.StoredDestinationLocation, Alpha=$Particles.StoredNormalizedExec)
    )) {isSetParameters=true, module=SetVariables_#9}
   /SetVariables(Particles.SpritePosition:Vector3f=&Add_Vector(A=$Particles.Initial.SpriteOffset, B=$Particles.RibbonPosition)) {isSetParameters=true, module=SetVariables_#10}
   /SetVariables(Particles.SpriteColor=&MultiplyLinearColor(
     Linear Color=$Particles.Initial.SpriteColor
     Scale Factor=&MakeVector4_FromFloat(Float=&FloatFromCurve(FloatCurve=DataInterface() {propertyValues {Curve {keys [
           {arriveTangent=-2.4415, interpMode=RCIM_Cubic, time=0, value=0}
           <at 0.175 1 RCIM_Cubic>
           {arriveTangent=-3.0397, interpMode=RCIM_Cubic, leaveTangent=-3.0397, tangentMode=RCTM_User, time=1, value=0}
          ]}}}))
    )) {isSetParameters=true, module=SetVariables_#11}
   &ScaleRibbonWidth(Ribbon Width Scale=&FloatFromCurve(FloatCurve=DataInterface() {propertyValues {Curve {keys [
         {arriveTangent=-4.302, interpMode=RCIM_Cubic, time=0, value=1}
         {arriveTangent=-2.5253, interpMode=RCIM_Cubic, leaveTangent=-11.4402, tangentMode=RCTM_User, time=1, value=0}
        ]}}}))
   /SetVariables(
    Particles.RibbonWidthBlue:NiagaraFloat=&Multiply_Float(
     A=$Particles.RibbonWidth
     B=&Lerp_Float(A=0, B=1, Alpha=&MakeFloatFromInt(Int=$User.VariantNum))
    )
    Particles.RibbonWidthPink:NiagaraFloat=&Multiply_Float(
     A=$Particles.RibbonWidth
     B=&Lerp_Float(A=1, B=0, Alpha=&MakeFloatFromInt(Int=$User.VariantNum))
    )
   ) {isSetParameters=true, module=SetVariables_#12}
   /SetVariables(
    Particles.LightColor=&MultiplyLinearColor(Linear Color=$Particles.SpriteColor, Scale Factor=&MakeVector4_FromFloat(Float=2))
    Particles.LightEnabled:NiagaraBool=&SetBoolByFloatComparison(
     A=$Particles.Initial.ParticleCurveMask
     B=&Lerp_Float(
      A=&RandomRangeFloat(Minimum=0.95, Maximum=0.952)
      B=1
      Alpha=&Round_Float(Float:NiagaraFloat=$Particles.Initial.MaterialRandom)
     )
    )
    Particles.LightPosition:Vector3f=&Add_Vector(A=$Particles.SpritePosition, B=<0 0 20>)
    Particles.LightRadius=40
   ) {isSetParameters=true, module=SetVariables_#13}
  ]
  rendererClasses [
   &NiagaraRibbonRendererProperties
   &NiagaraSpriteRendererProperties
   &NiagaraLightRendererProperties
  ]
  renderers [
   {rendererClass=&NiagaraRibbonRendererProperties}
   {rendererClass=&NiagaraSpriteRendererProperties, rendererIndex=1}
   {rendererClass=&NiagaraLightRendererProperties, rendererIndex=2}
   {rendererClass=&NiagaraRibbonRendererProperties, rendererIndex=3}
  ]
 }
 Electric_Shoulder_Bugaloo_Sockets {
  emitterSpawnScript [/SetVariables(Emitter.MaxEmitterLoops=12) {isSetParameters=true, module=SetVariables_#1}]
  emitterUpdateScript [
   &EmitterState(
    Life Cycle Mode=::NewEnumerator1
    Loop Behavior=::NewEnumerator2
    Loop Count:NiagaraInt32=$Emitter.MaxEmitterLoops
    Loop Duration=&RandomRangeFloat(Minimum=0.05, Maximum=0.1, Evaluation Type=::NewEnumerator1)
    Recalculate Duration Each Loop=true
    UseLoopDelay=false
   )
   &SpawnBurst_Instantaneous(
    Spawn Count=4
    Spawn Group:NiagaraInt32=&RandomRangeInt(Minimum=0, Maximum=9999, Evaluation Type=::NewEnumerator1)
    Use Spawn Probability=true
   )
  ]
  particleSpawnScript [
   /SetVariables(Particles.LeftOrRight:NiagaraInt32=/Game/Effects/Niagara/DynamicInputs/Arrays/FN_SelectIntFromArray(
     Array Sampling Mode:ENiagaraArraySamplingMode=::NewEnumerator1
     Interpolation:NiagaraFloat=?
     Evaluation Type:ENiagaraRandomnessEvaluation=::NewEnumerator0
     Normalized Interpolation:NiagaraBool=true
     Direct Array Index:NiagaraInt32=&RandomRangeInt(Minimum=0, Maximum=1, Randomness Mode=::NewEnumerator1, Random Seed=&SpawnGroup(), Fixed Random Seed=true)
     Recalculate Random Each Loop:NiagaraBool=false
     Int Selection Array:NiagaraDataInterfaceArrayInt32=DataInterface() {propertyValues {IntData [0,3]}}
     Randomness Mode:ENiagaraRandomnessMode=::NewEnumerator1
     Random Seed:NiagaraInt32=&SpawnGroup()
     Fixed Random Seed:NiagaraBool=true
    )) {isSetParameters=true}
   /SetVariables(Particles.StoredNormalizedExec:NiagaraFloat=&ReturnNormalizedExecIndex()) {isSetParameters=true, module=SetVariables_#2}
   /SetVariables(Particles.ParticleCurveMask:NiagaraFloat=&Sine(Normalized Angle=&ReturnNormalizedExecIndex(), Period=2, Scale=1, Bias=0)) {isSetParameters=true, module=SetVariables_#3}
   /SetVariables(
    Particles.ElectricHeight:NiagaraFloat=&RandomRangeFloat(Minimum=2, Maximum=8, Randomness Mode=::NewEnumerator2, Random Seed=&SpawnGroup(), Fixed Random Seed=true)
    Particles.ElectricSlide:NiagaraFloat=&Multiply_Float(
     A=&RandomRangeFloat(Minimum=2, Maximum=12, Randomness Mode=::NewEnumerator2, Random Seed=&SpawnGroup(), Fixed Random Seed=true)
     B=$Particles.ParticleCurveMask
    )
    Particles.StartBone:NiagaraInt32=&Add_Integer(
     A=&RandomRangeInt(Minimum=0, Maximum=2, Randomness Mode=::NewEnumerator1, Random Seed=&SpawnGroup(), Fixed Random Seed=true)
     B=$Particles.LeftOrRight
    )
   ) {isSetParameters=true, module=SetVariables_#4}
   /SetVariables(Particles.EndBone:NiagaraInt32=&Add_Integer(
     A=6
     B=&Add_Integer(
      A=&RandomRangeInt(
       Minimum=0
       Maximum=2
       Randomness Mode=::NewEnumerator1
       Random Seed=&Add_Integer(A=&SpawnGroup(), B=1)
       Fixed Random Seed=true
      )
      B=$Particles.LeftOrRight
     )
    )) {isSetParameters=true}
   /SetVariables(
    Particles.SpriteColor=&MakeLinearColorFromVectorAndFloat(
     "Vector (RGB)"=&MakeVectorFromLinearColorRGB(Color=&LerpLinearColors(
       StartColor=<rgba 1 0 0.6 1>
       EndColor=<rgba 0 0.6333 1 1>
       LerpFactor=&MakeFloatFromInt(Int=$User.VariantNum)
      ))
     "Float (Alpha)"=&RandomRangeFloat(Minimum=0.1, Maximum=0.145)
    )
    Particles.SpriteOffset:Vector3f=<3 0 0>
    Particles.SpriteSize=&Vector2DFromFloat(Value=&UniformAOrBFloat(A=25, B=0, Distribution Weight=&Ceil_Float(Float=$Particles.ParticleCurveMask)))
   ) {isSetParameters=true, module=SetVariables_#5}
   &InitializeRibbon(
    Color=&MakeLinearColor(
     R:NiagaraFloat=&RandomRangeFloat(Minimum=3, Maximum=9)
     G:NiagaraFloat=0
     B:NiagaraFloat=0
     A:NiagaraFloat=0
    )
    Lifetime=&RandomRangeFloat(Minimum=0.05, Maximum=0.2)
    Module.Write Color=true
    Module.Write Lifetime=true
    Module.Write Mass=false
    Module.Write Material Random=true
    Module.Write Position=false
    Module.Write RibbonTwist=false
    Module.Write RibbonWidth=true
    Ribbon Width=&RandomRangeFloat(Minimum=2, Maximum=6)
   )
   /SetVariables(Particles.RibbonID=&MakeNiagaraIDFromInt(Index:NiagaraInt32=&SpawnGroup(), Acquire Tag:NiagaraInt32=0)) {isSetParameters=true, module=SetVariables_#6}
   &DynamicMaterialParameters(
    Index 0 Param 1=1
    Index 1 Param 4=6.666
    Module.Param0WriteEnabled=true
    Module.Param0WriteEnabled1=true
    Module.Param1WriteEnabled=true
    Module.Param1WriteEnabled1=true
    Module.Param2WriteEnabled=true
    Module.Param2WriteEnabled1=true
    Module.Param3WriteEnabled=true
    Module.Param3WriteEnabled1=true
    Write Parameter Index 1=true
   )
   &CameraOffset(Camera Offset Amount=10)
  ]
  particleUpdateScript [
   &SkeletalMeshLocation(
    Apply To Particle Position=false
    Bone / Socket Index=$Particles.StartBone
    Inherit Velocity=false
    MaskSpawnGroup=false
    Mesh Sampling Type=::NewEnumerator5
    Skeletal Mesh=$User.SkeletalMesh
    Socket Sampling Mode=::NewEnumerator5
    Use Sampled Mesh Attribute as Ribbon ID=false
   )
   /SetVariables(Particles.StoredSourceLocation:Vector3f=$Output.SkeletalMeshLocation.SampledPosition) {isSetParameters=true, module=SetVariables_#7}
   &SkeletalMeshLocation(
    Apply To Particle Position=false
    Bone / Socket Index=$Particles.EndBone
    Inherit Velocity=false
    MaskSpawnGroup=false
    Mesh Sampling Type=::NewEnumerator5
    Skeletal Mesh=$User.SkeletalMesh
    Socket Sampling Mode=::NewEnumerator5
    Use Sampled Mesh Attribute as Ribbon ID=false
   ) {module=SkeletalMeshLocation002}
   /SetVariables(Particles.StoredDestinationLocation:Vector3f=$Output.SkeletalMeshLocation002.SampledPosition) {isSetParameters=true, module=SetVariables_#8}
   &ParticleState()
   /SetVariables(Particles.RibbonPosition:Vector3f=&Add_Vector(
     A=&Multiply_VectorByFloat(
      Vector=&Add_Vector(
       A=&MakeVector(X=0, Y=0, Z=$Particles.Initial.ElectricHeight)
       B=&Multiply_VectorByFloat(
        Vector=&ReturnMeshOrientationAxis(Mesh Orientation Axis=::XAxis, Mesh Orientation Quat=$Output.SkeletalMeshLocation.SampledRotation)
        Float=$Particles.Initial.ElectricSlide
       )
      )
      Float=&Multiply_Float(
       A=&ClampFloat(Float=&Power(a=$Particles.NormalizedAge, b=0.75), Min=1)
       B=$Particles.Initial.ParticleCurveMask
      )
     )
     B=&Lerp_Vector(A=$Particles.StoredSourceLocation, B=$Particles.StoredDestinationLocation, Alpha=$Particles.StoredNormalizedExec)
    )) {isSetParameters=true, module=SetVariables_#9}
   /SetVariables(Particles.SpritePosition:Vector3f=&Add_Vector(A=$Particles.Initial.SpriteOffset, B=$Particles.RibbonPosition)) {isSetParameters=true, module=SetVariables_#10}
   /SetVariables(Particles.SpriteColor=&MultiplyLinearColor(
     Linear Color=$Particles.Initial.SpriteColor
     Scale Factor=&MakeVector4_FromFloat(Float=&FloatFromCurve(FloatCurve=DataInterface() {propertyValues {Curve {keys [
           {arriveTangent=-2.4415, interpMode=RCIM_Cubic, time=0, value=0}
           <at 0.175 1 RCIM_Cubic>
           {arriveTangent=-3.0397, interpMode=RCIM_Cubic, leaveTangent=-3.0397, tangentMode=RCTM_User, time=1, value=0}
          ]}}}))
    )) {isSetParameters=true, module=SetVariables_#11}
   &ScaleRibbonWidth(Ribbon Width Scale=&FloatFromCurve(FloatCurve=DataInterface() {propertyValues {Curve {keys [
         {arriveTangent=-0.8857, interpMode=RCIM_Cubic, time=0, value=1}
         {arriveTangent=-3.3118, interpMode=RCIM_Cubic, leaveTangent=-3.3118, tangentMode=RCTM_User, time=1, value=0}
        ]}}}))
   /SetVariables(
    Particles.RibbonWidthBlue:NiagaraFloat=&Multiply_Float(
     A=$Particles.RibbonWidth
     B=&Lerp_Float(A=0, B=1, Alpha=&MakeFloatFromInt(Int=$User.VariantNum))
    )
    Particles.RibbonWidthPink:NiagaraFloat=&Multiply_Float(
     A=$Particles.RibbonWidth
     B=&Lerp_Float(A=1, B=0, Alpha=&MakeFloatFromInt(Int=$User.VariantNum))
    )
   ) {isSetParameters=true, module=SetVariables_#12}
  ]
  rendererClasses [&NiagaraRibbonRendererProperties, &NiagaraSpriteRendererProperties]
  renderers [
   {rendererClass=&NiagaraRibbonRendererProperties}
   {rendererClass=&NiagaraSpriteRendererProperties, rendererIndex=1}
   {rendererClass=&NiagaraRibbonRendererProperties, rendererIndex=2}
  ]
 }
 NE_GlowDecal {
  emitterUpdateScript [
   &EmitterState(Life Cycle Mode=::NewEnumerator1, Loop Behavior=::NewEnumerator1, Loop Duration=1, UseLoopDelay=false)
   &SpawnBurst_Instantaneous(Spawn Count=16, Use Spawn Probability=false)
  ]
  particleSpawnScript [
   /SetVariables() {isSetParameters=true}
   /Niagara/Modules/Spawn/Initialization/InitializeParticle(
    Color=&LerpLinearColors(
     StartColor=<rgba 3.3 0 0.99 1>
     EndColor=&MultiplyLinearColor(Linear Color=<rgba 0 1.1 3.3 1>, Scale Factor=&MakeVector4_FromFloat(Float=1))
     LerpFactor=&MakeFloatFromInt(Int=$User.VariantNum)
    )
    Lifetime=&RandomRangeFloat(Minimum=0.85, Maximum=1)
    Module.Write Color=true
    Module.Write Lifetime=true
    Module.Write Mass=true
    Module.Write Material Random=true
    Module.Write Position=false
    Module.Write Scale=false
    Module.Write SpriteRotation=false
    Module.Write SpriteSize=true
    Position=&SimulationPosition()
    Sprite Size=&Vector2DFromFloat(Value=&RandomRangeFloat(Minimum=20, Maximum=80))
   )
   &CameraOffset(Camera Offset Amount=6)
   &SphereLocation(
    MaskSpawnGroup=false
    Override Seed=false
    Sphere Coordinate Space=::Simulation
    Sphere Radius=5
    Surface Only=false
    UseNonUniformScale=false
    UseOffset=false
    UseSphereOrientation=false
   )
   &DynamicMaterialParameters(
    Index 0 Param 1=2
    Module.Param0WriteEnabled=true
    Module.Param1WriteEnabled=true
    Module.Param2WriteEnabled=true
    Module.Param3WriteEnabled=true
   )
  ]
  particleUpdateScript [
   &ParticleState()
   &SkeletalMeshLocation(
    Inherit Velocity=false
    Interpolate Mesh Sampling=false
    MaskSpawnGroup=false
    Module.Override Randomness=false
    Sampled Position Offset Coordinate Space:ENiagaraMeshTransforms=::NewEnumerator3
    Skeletal Mesh=$User.SkeletalMesh
    Use Sampled Mesh Attribute as Ribbon ID=false
   ) {module=SkeletalMeshLocation001}
   &SolveForcesAndVelocity(Clamp Velocity=false, Limit Acceleration=false)
   &DynamicMaterialParameters(Module.Param0WriteEnabled=true, Module.Param1WriteEnabled=true, Module.Param2WriteEnabled=true, Module.Param3WriteEnabled=true) {enabled=false, module=DynamicMaterialParameters001}
   &ScaleColor(
    Scale Alpha=&FloatFromCurve(FloatCurve=DataInterface() {propertyValues {Curve {keys [
         {arriveTangent=3.4557, interpMode=RCIM_Cubic, time=0, value=0.8}
         {arriveTangent=-0.3202, interpMode=RCIM_Cubic, leaveTangent=-0.3202, tangentMode=RCTM_User, time=1, value=0}
        ]}}})
    ScaleA=true
    ScaleRGB=false
   )
   &ScaleSpriteSize(Scale Factor=&Vector2DFromFloat(Value=&FloatFromCurve(FloatCurve=DataInterface() {propertyValues {Curve {keys [<at 0 1>,<at 1 0.5>]}}})))
  ]
  rendererClasses [&NiagaraSpriteRendererProperties]
  renderers [{rendererClass=&NiagaraSpriteRendererProperties}]
 }
 Sparkle_Burst {
  emitterUpdateScript [
   &EmitterState(Life Cycle Mode=::NewEnumerator1, Loop Behavior=::NewEnumerator2, Loop Duration=0.1, UseLoopDelay=false)
   &SpawnBurst_Instantaneous(Spawn Count=25, Use Spawn Probability=false)
  ]
  particleSpawnScript [
   &InitializeParticle(
    Lifetime=&RandomRangeFloat(Minimum=0.2, Maximum=0.8)
    Module.Write Color=true
    Module.Write Lifetime=true
    Module.Write Mass=false
    Module.Write Material Random=true
    Module.Write Position=false
    Module.Write Scale=false
    Module.Write SpriteRotation=true
    Module.Write SpriteSize=true
    Sprite Rotation=&RandomRangeFloat(Minimum=-180, Maximum=180)
    Sprite Size=&Vector2DFromFloat(Value=&RandomRangeFloat(Minimum=1.5, Maximum=2.6))
   )
   /SetVariables(Particles.SpriteSizeAlt:Vector2f=&Vector2DFromFloat(Value=&RandomRangeFloat(Minimum=2.5, Maximum=6))) {isSetParameters=true}
   /SetVariables(
    Particles.Rand01:NiagaraFloat=&UniformAOrBFloat(A=0)
    Particles.Rand02:NiagaraFloat=&UniformAOrBFloat(A=0)
    Particles.SpawnGroup=&RandomRangeInt(Minimum=0, Maximum=100)
   ) {isSetParameters=true}
   &SkeletalMeshLocation(Inherit Velocity=false, MaskSpawnGroup=false, Module.Override Randomness=false, Skeletal Mesh=$User.SkeletalMesh) {module=SkeletalMeshLocation001}
   &SphereLocation(
    MaskSpawnGroup=false
    Override Seed=false
    Sphere Radius=10
    Surface Only=false
    UseNonUniformScale=false
    UseOffset=false
    UseSphereOrientation=false
   )
   &InheritVelocity(Module.Limit Speed=true, Velocity Limit=200)
   &AddVelocityFromPoint(
    Module.Use Falloff Distance=false
    Origin Offset=<0 0 100>
    Velocity Strength=&RandomRangeFloat(Minimum=150, Maximum=500)
   )
  ]
  particleUpdateScript [
   &ParticleState()
   &Drag(
    Drag=&RandomRangeFloat(
     Minimum=HlslExpression() {hlslExpression="3 * (1 + Emitter.LoopCount)"}
     Maximum=HlslExpression() {hlslExpression="6 * (1 + Emitter.LoopCount)"}
    )
    Use Linear Drag=true
    Use Rotational Drag=false
   )
   &GravityForce(Gravity=&MakeVector(X=0, Y=0, Z=&RandomRangeFloat(Minimum=0, Maximum=-700)))
   &CurlNoiseForce(Noise Frequency=15, Noise Strength=500, Pan Noise=true, Pan Noise Field=&RandomVector(Vector Scale=5))
   &SolveForcesAndVelocity(Clamp Velocity=false, Limit Acceleration=false)
   /SetVariables(
    Particles.PositionAlt:Vector3f=&Add_Vector(
     A=&Lerp_Vector(
      A=Linked() {linkedVariable {name=Particles.Initial.Position, type=NiagaraPosition}}
      B=Linked() {linkedVariable {name=Particles.Position, type=NiagaraPosition}}
      Alpha=0.5
     )
     B=&Lerp_Vector(
      A=<0 0 0>
      B=&RandomVector(Vector Scale=&RandomRangeFloat(Minimum=0, Maximum=20))
      Alpha=$Particles.NormalizedAge
     )
    )
    Particles.SparklesColor:LinearColor=&HueShiftLinearColor(
     Hue=&RandomRangeFloat(Minimum=0, Maximum=0)
     Saturation=&RandomRangeFloat(Minimum=0.6, Maximum=1)
     Value=&FloatFromCurve(
      FloatCurve=DataInterface() {propertyValues {Curve {keys [{arriveTangent=3.1648, interpMode=RCIM_Cubic, leaveTangent=3.1648, tangentMode=RCTM_User, time=0, value=1}, <at 0.5 0>]}}}
      Scale Curve=&RandomRangeFloat(Minimum=0.5, Maximum=10)
     )
     Linear Color=&LerpLinearColors(
      StartColor=&LerpLinearColors(
       StartColor=<rgba 6.3424 0 10 1>
       EndColor=&LerpLinearColors(StartColor=<rgba 18.2078 0 20 1>, EndColor=<rgba 10 0 3.1297 1>, LerpFactor=$Particles.Rand01)
       LerpFactor=$Particles.Rand02
      )
      EndColor=&LerpLinearColors(
       StartColor=<rgba 0 4.2855 10 1>
       EndColor=&LerpLinearColors(StartColor=<rgba 0 4.3841 20 1>, EndColor=<rgba 0.6409 0 10 1>, LerpFactor=$Particles.Rand01)
       LerpFactor=$Particles.Rand02
      )
      LerpFactor=&MakeFloatFromInt(Int=$User.VariantNum)
     )
    )
    Particles.SpriteRotation:NiagaraFloat=&Add_Float(
     A=$Particles.Initial.SpriteRotation
     B=&Lerp_Float(A=0, B=&RandomRangeFloat(Minimum=-200, Maximum=200), Alpha=$Particles.NormalizedAge)
    )
   ) {isSetParameters=true}
   &ScaleSpriteSize(Scale Factor=&Vector2DFromFloat(Value=&FloatFromCurve(FloatCurve=DataInterface() {propertyValues {Curve {keys [{interpMode=RCIM_Cubic, tangentMode=RCTM_User, time=0.6, value=1}, <at 1 0>]}}})))
  ]
  rendererClasses [&NiagaraSpriteRendererProperties]
  renderers [{rendererClass=&NiagaraSpriteRendererProperties}]
 }
}
summary {
 systemName=NS_Skin_FoxWarrior_Burst
 emitters [
  {
   emitterName=Electric_Shoulder_Bugaloo_Sockets
   rendererClasses [&NiagaraRibbonRendererProperties, &NiagaraSpriteRendererProperties]
  }
  {
   emitterName=Electric_Bone_Explosion
   rendererClasses [
    &NiagaraRibbonRendererProperties
    &NiagaraSpriteRendererProperties
    &NiagaraLightRendererProperties
   ]
  }
  {emitterName=NE_GlowDecal, rendererClasses [&NiagaraSpriteRendererProperties]}
  {
   emitterName=Electric_Bone_Explosion001
   rendererClasses [
    &NiagaraRibbonRendererProperties
    &NiagaraSpriteRendererProperties
    &NiagaraLightRendererProperties
   ]
  }
  {emitterName=Sparkle_Burst, rendererClasses [&NiagaraSpriteRendererProperties]}
 ]
 userVariables [
  User.SkeletalMesh {
   type=NiagaraDataInterfaceSkeletalMesh
   defaultValue {
    struct=NiagaraExt_VariableValue_DataInterface
    value {dataInterface=&self:NiagaraDataInterfaceSkeletalMesh_0}
   }
  }
  User.VariantNum {type=NiagaraInt32}
 ]
}
