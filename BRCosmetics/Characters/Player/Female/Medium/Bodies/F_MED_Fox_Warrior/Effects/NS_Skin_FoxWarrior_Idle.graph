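# Idle Niagara effect for the Fox Warrior skin that flickers short electric arcs, apparently at the shoulders, between bones on the user-supplied skeletal mesh.
# Randomly timed six-particle bursts share a random spawn group, which seeds side, bone pair and bend and sets ribbon ID.
# Particles lerp between the sampled bones with a sine-masked bow, driving two ribbon renderers, glow sprites and lights.
# VariantNum switches pink to blue through color and ribbon widths.
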NiagaraSystem /BRCosmetics/Characters/Player/Female/Medium/Bodies/F_MED_Fox_Warrior/Effects/NS_Skin_FoxWarrior_Idle
dependencies {
 usedDataInterfaces [&NiagaraDataInterfaceArrayInt32, &NiagaraDataInterfaceCurve]
 usedDynamicInputs []
 usedModules [&SystemState]
}
emitters {Electric_Shoulder_Bugaloo {
  emitterUpdateScript [
   &EmitterState(
    Life Cycle Mode=::NewEnumerator1
    Loop Duration=&RandomRangeFloat(Minimum=0.05, Maximum=0.45, Evaluation Type=::NewEnumerator1)
    Recalculate Duration Each Loop=true
    UseLoopDelay=false
   )
   &SpawnBurst_Instantaneous(
    Spawn Count=6
    Spawn Group:NiagaraInt32=&RandomRangeInt(Minimum=0, Maximum=9999, Evaluation Type=::NewEnumerator1)
    Use Spawn Probability=false
   )
  ]
  particleSpawnScript [
   /SetVariables(Particles.LeftOrRight:NiagaraInt32=/Game/Effects/Niagara/DynamicInputs/Arrays/FN_SelectIntFromArray(
     Array Sampling Mode:ENiagaraArraySamplingMode=::NewEnumerator1
     Interpolation:NiagaraFloat=?
     Evaluation Type:ENiagaraRandomnessEvaluation=::NewEnumerator0
     Normalized Interpolation:NiagaraBool=true
     Direct Array Index:NiagaraInt32=&RandomRangeInt(Minimum=0, Maximum=1, Randomness Mode=::NewEnumerator1, Random Seed=&SpawnGroup(), Fixed Random Seed=true)
     Recalculate Random Each Loop:NiagaraBool=false
     Int Selection Array:NiagaraDataInterfaceArrayInt32=DataInterface() {propertyValues {IntData [0,3]}}
     Randomness Mode:ENiagaraRandomnessMode=::NewEnumerator1
     Random Seed:NiagaraInt32=&SpawnGroup()
     Fixed Random Seed:NiagaraBool=true
    )) {isSetParameters=true}
   /SetVariables(Particles.ParticleCurveMask:NiagaraFloat=&Sine(Normalized Angle=&ReturnNormalizedExecIndex(), Period=2, Scale=1, Bias=0)) {isSetParameters=true}
   /SetVariables(
    Particles.ElectricHeight:NiagaraFloat=&RandomRangeFloat(Minimum=-5, Maximum=5, Randomness Mode=::NewEnumerator2, Random Seed=&SpawnGroup(), Fixed Random Seed=true)
    Particles.ElectricSlide:NiagaraFloat=&Multiply_Float(
     A=&RandomRangeFloat(Minimum=5, Maximum=9, Randomness Mode=::NewEnumerator2, Random Seed=&SpawnGroup(), Fixed Random Seed=true)
     B=$Particles.ParticleCurveMask
    )
    Particles.StartBone:NiagaraInt32=&Add_Integer(
     A=&RandomRangeInt(Minimum=0, Maximum=2, Randomness Mode=::NewEnumerator1, Random Seed=&SpawnGroup(), Fixed Random Seed=true)
     B=$Particles.LeftOrRight
    )
   ) {isSetParameters=true}
   /SetVariables(Particles.EndBone:NiagaraInt32=&Add_Integer(
     A=6
     B=&Add_Integer(
      A=&RandomRangeInt(
       Minimum=0
       Maximum=2
       Randomness Mode=::NewEnumerator1
       Random Seed=&Add_Integer(A=&SpawnGroup(), B=1)
       Fixed Random Seed=true
      )
      B=$Particles.LeftOrRight
     )
    )) {isSetParameters=true}
   /SetVariables(
    Particles.SpriteColor=&MakeLinearColorFromVectorAndFloat(
     "Vector (RGB)"=&MakeVectorFromLinearColorRGB(Color=&LerpLinearColors(
       StartColor=<rgba 1 0 0.6 1>
       EndColor=<rgba 0 0.6333 1 1>
       LerpFactor=&MakeFloatFromInt(Int=$User.VariantNum)
      ))
     "Float (Alpha)"=&RandomRangeFloat(Minimum=0.1, Maximum=0.145)
    )
    Particles.SpriteOffset:Vector3f=<3 0 0>
    Particles.SpriteSize=&Vector2DFromFloat(Value=&UniformAOrBFloat(A=25, B=0, Distribution Weight=&Ceil_Float(Float=$Particles.ParticleCurveMask)))
   ) {isSetParameters=true}
   &InitializeRibbon(
    Color=&MakeLinearColor(
     R:NiagaraFloat=&RandomRangeFloat(Minimum=5, Maximum=30)
     G:NiagaraFloat=1
     B:NiagaraFloat=1
     A:NiagaraFloat=1
    )
    Lifetime=&RandomRangeFloat(Minimum=0.15, Maximum=0.2, Randomness Mode=::NewEnumerator1, Random Seed=$Emitter.LoopCount, Fixed Random Seed=true)
    Module.Write Color=true
    Module.Write Lifetime=true
    Module.Write Mass=false
    Module.Write Material Random=true
    Module.Write Position=false
    Module.Write RibbonTwist=false
    Module.Write RibbonWidth=true
    Ribbon Width=&RandomRangeFloat(Minimum=3, Maximum=5)
   )
   /SetVariables(Particles.StoredNormalizedExec:NiagaraFloat=&ReturnNormalizedExecIndex()) {isSetParameters=true}
   /SetVariables(Particles.RibbonID=/Niagara/DynamicInputs/NiagaraID/MakeNiagaraIDFromInt(Index:NiagaraInt32=&SpawnGroup(), Acquire Tag:NiagaraInt32=0)) {isSetParameters=true}
   &DynamicMaterialParameters(
    Index 0 Param 1=1
    Index 1 Param 4=6.666
    Module.Param0WriteEnabled=true
    Module.Param0WriteEnabled1=true
    Module.Param1WriteEnabled=true
    Module.Param1WriteEnabled1=true
    Module.Param2WriteEnabled=true
    Module.Param2WriteEnabled1=true
    Module.Param3WriteEnabled=true
    Module.Param3WriteEnabled1=true
    Write Parameter Index 1=true
   )
   &CameraOffset(Camera Offset Amount=10)
  ]
  particleUpdateScript [
   &SkeletalMeshLocation(
    Apply To Particle Position=false
    Bone / Socket Index=$Particles.StartBone
    Inherit Velocity=false
    MaskSpawnGroup=false
    Mesh Sampling Type=::NewEnumerator5
    Skeletal Mesh=$User.SkeletalMesh
    Socket Sampling Mode=::NewEnumerator5
    Use Sampled Mesh Attribute as Ribbon ID=false
   )
   /SetVariables(Particles.StoredSourceLocation:Vector3f=$Output.SkeletalMeshLocation.SampledPosition) {isSetParameters=true}
   &SkeletalMeshLocation(
    Apply To Particle Position=false
    Bone / Socket Index=$Particles.EndBone
    Inherit Velocity=false
    MaskSpawnGroup=false
    Mesh Sampling Type=::NewEnumerator5
    Skeletal Mesh=$User.SkeletalMesh
    Socket Sampling Mode=::NewEnumerator5
    Use Sampled Mesh Attribute as Ribbon ID=false
   ) {module=SkeletalMeshLocation002}
   /SetVariables(Particles.StoredDestinationLocation:Vector3f=$Output.SkeletalMeshLocation002.SampledPosition) {isSetParameters=true}
   &ParticleState()
   /SetVariables(Particles.RibbonPosition:Vector3f=&Add_Vector(
     A=&Multiply_VectorByFloat(
      Vector=&Add_Vector(
       A=&MakeVector(X=0, Y=0, Z=$Particles.Initial.ElectricHeight)
       B=&Multiply_VectorByFloat(
        Vector=&ReturnMeshOrientationAxis(Mesh Orientation Axis=::XAxis, Mesh Orientation Quat=$Output.SkeletalMeshLocation.SampledRotation)
        Float=$Particles.Initial.ElectricSlide
       )
      )
      Float=&Multiply_Float(
       A=&ClampFloat(Float=&Power(a=$Particles.NormalizedAge, b=0.5), Min=0.35, Max=0.88)
       B=$Particles.Initial.ParticleCurveMask
      )
     )
     B=&Lerp_Vector(A=$Particles.StoredSourceLocation, B=$Particles.StoredDestinationLocation, Alpha=$Particles.StoredNormalizedExec)
    )) {isSetParameters=true}
   /SetVariables(Particles.SpritePosition:Vector3f=&Add_Vector(A=$Particles.Initial.SpriteOffset, B=$Particles.RibbonPosition)) {isSetParameters=true}
   /SetVariables(Particles.SpriteColor=&MultiplyLinearColor(
     Linear Color=$Particles.Initial.SpriteColor
     Scale Factor=&MakeVector4_FromFloat(Float=&FloatFromCurve(
       FloatCurve=DataInterface() {propertyValues {Curve {keys [
           {arriveTangent=-2.4415, interpMode=RCIM_Cubic, time=0, value=0}
           <at 0.175 1 RCIM_Cubic>
           {arriveTangent=-3.0397, interpMode=RCIM_Cubic, leaveTangent=-3.0397, tangentMode=RCTM_User, time=1, value=0}
          ]}}}
       Scale Curve=1.1
      ))
    )) {isSetParameters=true}
   &ScaleRibbonWidth(Ribbon Width Scale=&FloatFromCurve(FloatCurve=DataInterface() {propertyValues {Curve {keys [<at 0 1 RCIM_Cubic>, {arriveTangent=-1.5435, interpMode=RCIM_Cubic, leaveTangent=-1.5435, tangentMode=RCTM_User, time=1, value=0}]}}}))
   /SetVariables(
    Particles.LightColor=&MultiplyLinearColor(Linear Color=$Particles.SpriteColor, Scale Factor=&MakeVector4_FromFloat(Float=0.65))
    Particles.LightEnabled:NiagaraBool=&SetBoolByFloatComparison(A=$Particles.Initial.StoredNormalizedExec, B=0, Comparison Type=::NewEnumerator4)
    Particles.LightPosition:Vector3f=&Add_Vector(A=$Particles.SpritePosition, B=<-5 0 0>)
    Particles.LightRadius=50
   ) {isSetParameters=true}
   /SetVariables(
    Particles.RibbonWidthBlue:NiagaraFloat=&Multiply_Float(
     A=$Particles.RibbonWidth
     B=&Lerp_Float(A=0, B=4.2932, Alpha=&MakeFloatFromInt(Int=$User.VariantNum))
    )
    Particles.RibbonWidthPink:NiagaraFloat=&Multiply_Float(
     A=$Particles.RibbonWidth
     B=&Lerp_Float(A=1, B=$Particles.Initial.ParticleCurveMask, Alpha=&MakeFloatFromInt(Int=$User.VariantNum))
    )
   ) {isSetParameters=true}
  ]
  rendererClasses [
   &NiagaraRibbonRendererProperties
   &NiagaraSpriteRendererProperties
   &NiagaraLightRendererProperties
  ]
  renderers [
   {rendererClass=&NiagaraRibbonRendererProperties}
   {rendererClass=&NiagaraSpriteRendererProperties, rendererIndex=1}
   {rendererClass=&NiagaraLightRendererProperties, rendererIndex=2}
   {rendererClass=&NiagaraRibbonRendererProperties, rendererIndex=3}
  ]
 }}
summary {
 systemName=NS_Skin_FoxWarrior_Idle
 emitters [{
   emitterName=Electric_Shoulder_Bugaloo
   rendererClasses [
    &NiagaraRibbonRendererProperties
    &NiagaraSpriteRendererProperties
    &NiagaraLightRendererProperties
   ]
  }]
 userVariables [
  User.SkeletalMesh {
   type=NiagaraDataInterfaceSkeletalMesh
   defaultValue {
    struct=NiagaraExt_VariableValue_DataInterface
    value {dataInterface=&self:NiagaraDataInterfaceSkeletalMesh_0}
   }
  }
  User.VariantNum {type=NiagaraInt32}
 ]
}
