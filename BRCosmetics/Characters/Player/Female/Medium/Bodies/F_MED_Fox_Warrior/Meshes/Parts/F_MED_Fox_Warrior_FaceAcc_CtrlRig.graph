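# Control Rig for the female Fox Warrior's face accessory that shifts the mask bones in response to cheek movement.
# On forward solve, each inner cheek bone's height relative to faceAttach is remapped and clamped into a weight.
# That weight scales an additive local X translation on the same side's dyn_mask bone, with a slightly larger offset on the right.
# Two rotator-to-quaternion nodes sit unconnected in the graph.
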ControlRigBlueprint /BRCosmetics/Characters/Player/Female/Medium/Bodies/F_MED_Fox_Warrior/Meshes/Parts/F_MED_Fox_Warrior_FaceAcc_CtrlRig
graphs [
 RigVMFunctionLibrary {links [], nodes []}
 RigVMModel {nodes [
   RigUnit_BeginExecution = &RigUnit_BeginExecution()
   /Script/ControlRig.RigUnit_ModifyBoneTransforms(
    BoneToModify=((Bone="dyn_mask_l",Transform=(Rotation=(X=0,Y=0,Z=0,W=1),Translation=(X=1.3,Y=0,Z=0),Scale3D=(X=1,Y=1,Z=1))))
    Mode=AdditiveLocal
    WeightMaximum=1
    WeightMinimum=0
    Weight=&RigVMFunction_MathFloatRemap(
     SourceMaximum=4.5
     SourceMinimum=3.2
     TargetMinimum=0
     bClamp=true
     Value=&RigUnit_GetRelativeBoneTransform(Bone=L_cheek_inner, Space=faceAttach).Transform.Translation.Z
    ).Result
    ExecutePin=@RigUnit_BeginExecution.ExecutePin
   ) {is=struct RigVMUnitNode}
   > &RigUnit_ModifyBoneTransforms(
    BoneToModify=((Bone="dyn_mask_r",Transform=(Rotation=(X=0,Y=0,Z=0,W=1),Translation=(X=1.5,Y=0,Z=0),Scale3D=(X=1,Y=1,Z=1))))
    Mode=AdditiveLocal
    WeightMaximum=1
    WeightMinimum=0
    Weight=&RigVMFunction_MathFloatRemap(
     SourceMaximum=4.5
     SourceMinimum=3.2
     TargetMinimum=0
     bClamp=true
     Value=&RigUnit_GetRelativeBoneTransform(Bone=R_cheek_inner, Space=faceAttach).Transform.Translation.Z
    ).Result
   )
   /Script/RigVM.RigVMFunction_MathQuaternionFromRotator(Rotator=(Pitch=0,Yaw=0,Roll=0)) {is=struct RigVMUnitNode}
   &RigVMFunction_MathQuaternionFromRotator(Rotator=(Pitch=-5,Yaw=0,Roll=-3))
  ]}
]
hierarchy [
 root {loc=<0 0 0>, rot=<0 0 0 1>}
 pelvis {parent=root, loc=<0 -0.5301 98.0571>, rot=<0.0127 0.707 -0.0127 -0.707>}
 spine_01 {parent=pelvis, loc=<6.1208 -0.9328 0>, rot=<0 0 0.034 -0.9994>}
 spine_02 {parent=spine_01, loc=<6.5996 -0.2095 0>, rot=<0 0 0.0238 -0.9997>}
 spine_03 {parent=spine_02, loc=<6.4914 -0.1237 0>, rot=<0 0 0.0149 -0.9999>}
 spine_04 {parent=spine_03, loc=<6.9598 0.0794 0>, rot=<0 0 -0.0132 -0.9999>}
 spine_05 {parent=spine_04, loc=<6.9987 0 0>, rot=<0 0 -0.0015 -1>}
 clavicle_l {parent=spine_05, loc=<5.4831 -0.1053 -3.2946>, rot=<0.0465 -0.7289 -0.009 -0.6829>}
 clavicle_r {parent=spine_05, loc=<5.4831 -0.1053 3.2946>, rot=<0.7289 0.0465 -0.6829 0.009>}
 neck_01 {parent=spine_05, loc=<11.3138 0.1529 0>, rot=<0 0 -0.2003 -0.9797>}
 neck_02 {parent=neck_01, loc=<4.5627 0 0>, rot=<0 0 0.0504 -0.9987>}
 head {parent=neck_02, loc=<5.3914 0 0>, rot=<0 0 0.115 -0.9934>}
 faceAttach {parent=head, loc=<-3.5442 -0.4617 -0.0134>, rot=<0.0082 -0.7071 -0.0082 -0.7071>, scale=<0.89 0.89 0.89>}
 L_lip_upper_outer {parent=faceAttach, loc=<1.3287 9.3628 1.1488>, rot=<0.036 -0.0728 0.2627 -0.9615>}
 L_lip_corner {parent=faceAttach, loc=<2.2147 8.844 1.1198>, rot=<0.0033 0.0076 -0.0147 -0.9999>}
 R_lip_corner {parent=faceAttach, loc=<-2.3008 8.7654 1.0868>, rot=<0.003 -0.0079 0.0146 -0.9999>}
 C_jaw {parent=faceAttach, loc=<-0.0407 3.0671 1.1699>, rot=<0 0 0 1>}
 C_lip_lower_mid {parent=C_jaw, loc=<0.0258 6.923 -0.3352>, rot=<0 -0.0225 -0.0048 -0.9997>}
 R_lip_lower_outer {parent=C_jaw, loc=<-1.3495 6.4509 -0.2156>, rot=<-0.008 0.0693 -0.2166 -0.9738>}
 R_lip_upper_outer {parent=faceAttach, loc=<-1.3728 9.393 1.2024>, rot=<0.0377 0.0679 -0.2609 -0.9623>}
 L_cheek_inner {parent=faceAttach, loc=<4.3659 8.3791 3.1802>, rot=<-0.0458 -0.2069 0.3264 -0.9212>}
 R_cheek_inner {parent=faceAttach, loc=<-4.3847 8.3882 3.1723>, rot=<-0.0464 0.2086 -0.3262 -0.9208>}
 dyn_collar_01_r {parent=head, loc=<-2.6953 -6.8963 1.581>, rot=<0.683 0.0028 0.7304 -0.003>}
 dyn_collar_01_l {parent=head, loc=<-2.6953 -6.8963 -1.581>, rot=<0.003 -0.7304 -0.0028 -0.683>}
 dyn_collar_02_l {parent=head, loc=<4.4098 -6.8376 -1.1654>, rot=<0.003 -0.7304 -0.0028 -0.683>}
 dyn_collar_02_r {parent=head, loc=<4.4098 -6.8376 1.1654>, rot=<0.683 0.0028 0.7304 -0.003>}
 dyn_ear_r {parent=head, loc=<15.1785 -1.8902 5.8067>, rot=<0.4995 0.2671 0.1111 -0.8166>}
 dyn_ear_l {parent=head, loc=<15.1789 -1.8902 -5.8067>, rot=<0.8166 -0.1111 -0.2671 -0.4995>}
 dyn_hair_r_MSTR {parent=head, loc=<2.7246 2.7682 7.3286>, rot=<-0.0437 -0.0055 0.9921 0.1177>}
 dyn_hair_01_r {parent=dyn_hair_r_MSTR, loc=<0 0 0>, rot=<0 0 0 1>}
 dyn_hair_02_r {parent=dyn_hair_01_r, loc=<6.4768 0 0>, rot=<0.014 -0.024 0.282 -0.959>}
 dyn_hair_l_MSTR {parent=head, loc=<2.725 2.7682 -7.3286>, rot=<-0.1177 -0.9918 0.0048 0.0496>}
 dyn_hair_01_l {parent=dyn_hair_l_MSTR, loc=<0 0 0>, rot=<0 0 0 -1>}
 dyn_hair_02_l {parent=dyn_hair_01_l, loc=<6.5068 0.0245 -0.0446>, rot=<-0.0556 0.1295 -0.2768 -0.9505>}
 dyn_mask_r {parent=head, loc=<-0.8866 7.059 3.8991>, rot=<0.6219 0.4249 0.6222 0.2135>}
 dyn_mask_l {parent=head, loc=<-0.8866 7.059 -3.8991>, rot=<-0.2135 -0.6222 -0.4249 -0.6219>}
]
