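# Physics asset for the Fox Warrior female medium body's hood part, providing capsule bodies and joints for secondary motion of collar, ears and hair.
# Four collar pieces and two ears attach directly to the head with angular limits, tighter on the collar.
# Each side's two-segment hair chain hangs from a master body that rotates freely on the head.
# The head joins the neck with free rotation, while both clavicle capsules have no constraints.
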PhysicsAsset /BRCosmetics/Characters/Player/Female/Medium/Bodies/F_MED_Fox_Warrior/Meshes/Parts/F_MED_Fox_Warrior_dyn_hood
bodies {
 clavicle_l [{shapeType=Capsule, length=5.75, radius=6.5, center=<5.1998 -1.4628 -1.4751>, rotation=<rot 90 0 0.0001>}]
 clavicle_r [{shapeType=Capsule, length=5.75, radius=6.5, center=<-5.1998 1.4628 1.4751>, rotation=<rot 90 0 -179.9999>}]
 dyn_collar_01_l [{shapeType=Capsule, length=4.75, radius=2.75, center=<7.8329 0 -0.2479>, rotation=<rot 79.9999 0 0>}]
 dyn_collar_01_r [{shapeType=Capsule, length=4.75, radius=2.75, center=<7.8329 0 -0.2479>, rotation=<rot 79.9999 0 0>}]
 dyn_collar_02_l [{shapeType=Capsule, length=5.25, radius=2.75, center=<8.047 0 -0.8623>, rotation=<rot 79.9999 0 0>}]
 dyn_collar_02_r [{shapeType=Capsule, length=5.25, radius=2.75, center=<8.047 0 -0.8623>, rotation=<rot 79.9999 0 0>}]
 dyn_ear_l [{shapeType=Capsule, length=2, radius=2, center=<2 0 0>, rotation=<rot 90 0 0>}]
 dyn_ear_r [{shapeType=Capsule, length=2, radius=2, center=<2 0 0>, rotation=<rot 90 0 0>}]
 dyn_hair_01_l [{shapeType=Capsule, length=3.75, radius=1, center=<3.3735 0.0001 0>, rotation=<rot 90 0 0>}]
 dyn_hair_01_r [{shapeType=Capsule, length=3.75, radius=1, center=<3.3735 0.0001 0>, rotation=<rot 90 0 0>}]
 dyn_hair_02_l [{shapeType=Capsule, length=2.5, radius=1, center=<3.6103 0 0.0001>, rotation=<rot 90 0 0>}]
 dyn_hair_02_r [{shapeType=Capsule, length=2.5, radius=1, center=<3.6103 0 0.0001>, rotation=<rot 90 0 0>}]
 dyn_hair_l_MSTR [{shapeType=Capsule, length=1, radius=1}]
 dyn_hair_r_MSTR [{shapeType=Capsule, length=1, radius=1}]
 head [
  {shapeType=Capsule, length=10.1982, radius=5.7796, center=<0.7471 3.6671 -0.0002>, rotation=<rot 90 0 0>}
  {shapeType=Capsule, length=15.75, radius=7, center=<0.1922 -5.1978 0>}
 ]
 neck_01 [{shapeType=Capsule, length=11.25, radius=7, center=<-1.5058 3.0395 0>}]
}
constraints [
 {bone1Name=dyn_collar_01_r, bone2Name=head, swing1LimitDegrees=10, swing2LimitDegrees=15, twistLimitDegrees=2}
 {bone1Name=dyn_collar_01_l, bone2Name=head, swing1LimitDegrees=10, swing2LimitDegrees=15, twistLimitDegrees=2}
 {bone1Name=dyn_collar_02_l, bone2Name=head, swing1LimitDegrees=10, swing2LimitDegrees=15, twistLimitDegrees=2}
 {bone1Name=dyn_collar_02_r, bone2Name=head, swing1LimitDegrees=10, swing2LimitDegrees=15, twistLimitDegrees=2}
 {bone1Name=dyn_ear_r, bone2Name=head, swing1LimitDegrees=30, swing2LimitDegrees=30, twistLimitDegrees=5}
 {bone1Name=dyn_ear_l, bone2Name=head, swing1LimitDegrees=30, swing2LimitDegrees=30, twistLimitDegrees=5}
 {
  bone1Name=dyn_hair_r_MSTR
  bone2Name=head
  swing1LimitDegrees=45
  swing1Motion=Free
  swing2LimitDegrees=45
  swing2Motion=Free
  twistLimitDegrees=45
  twistMotion=Free
 }
 {bone1Name=dyn_hair_01_r, bone2Name=dyn_hair_r_MSTR, swing1LimitDegrees=20, swing2LimitDegrees=20, twistLimitDegrees=2}
 {bone1Name=dyn_hair_02_r, bone2Name=dyn_hair_01_r, swing1LimitDegrees=20, swing2LimitDegrees=20, twistLimitDegrees=2}
 {
  bone1Name=dyn_hair_l_MSTR
  bone2Name=head
  swing1LimitDegrees=45
  swing1Motion=Free
  swing2LimitDegrees=45
  swing2Motion=Free
  twistLimitDegrees=45
  twistMotion=Free
 }
 {bone1Name=dyn_hair_01_l, bone2Name=dyn_hair_l_MSTR, swing1LimitDegrees=20, swing2LimitDegrees=20, twistLimitDegrees=2}
 {bone1Name=dyn_hair_02_l, bone2Name=dyn_hair_01_l, swing1LimitDegrees=20, swing2LimitDegrees=20, twistLimitDegrees=2}
 {
  bone1Name=head
  bone2Name=neck_01
  swing1LimitDegrees=45
  swing1Motion=Free
  swing2LimitDegrees=45
  swing2Motion=Free
  twistLimitDegrees=45
  twistMotion=Free
 }
]
